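# Sound cue for the PalmTree pickaxe cosmetic's impact on an enemy, using the shared pickaxe impact attenuation and impacts sound class.
# After a fixed short delay, a mixer layers two modulated random branches: one of six main impact waves and an optional sweetener from nine waves.
# Equally weighted empty slots interleave the sweeteners, so that quieter, slightly pitched-down layer plays on about half of hits.
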SoundCue /BRCosmetics/Sounds/PickAxes/PalmTree_IP/PickaxeImpactEnemy_PalmTree_Rnd_Cue
root=Delay_0
nodes [
 {id=Delay_0, children [Mixer_0], properties {delayMax=0.1, delayMin=0.1}}
 {id=Mixer_0, children [Modulator_0,Modulator_1], properties {inputVolume [1,1]}}
 {id=Modulator_0, children [Random_0]}
 {id=Modulator_1, children [Random_1], properties {pitchMax=1, pitchMin=0.9, volumeMax=0.75, volumeMin=0.65}}
 {id=Random_0, children [WavePlayer_0,WavePlayer_1,WavePlayer_2,WavePlayer_3,WavePlayer_4,WavePlayer_5], properties {weights [1,1,1,1,1,1]}}
 {
  id=Random_1
  children [
   WavePlayer_6
   null
   WavePlayer_7
   null
   WavePlayer_8
   null
   WavePlayer_9
   null
   WavePlayer_10
   null
   WavePlayer_11
   null
   WavePlayer_12
   null
   WavePlayer_13
   null
   WavePlayer_14
   null
  ]
  properties {weights [1,1,1,1,1,1,1,1,1,1,1,1,1,1,1,1,1,1]}
 }
 {id=WavePlayer_0, properties {soundWaveAssetPtr=&/IP_Pickaxe_PalmTree_Impact_01}}
 {id=WavePlayer_1, properties {soundWaveAssetPtr=&/IP_Pickaxe_PalmTree_Impact_02}}
 {id=WavePlayer_10, properties {soundWaveAssetPtr=&/IP_Pickaxe_PalmTree_ImpactSweetener_05}}
 {id=WavePlayer_11, properties {soundWaveAssetPtr=&/IP_Pickaxe_PalmTree_ImpactSweetener_06}}
 {id=WavePlayer_12, properties {soundWaveAssetPtr=&/IP_Pickaxe_PalmTree_ImpactSweetener_07}}
 {id=WavePlayer_13, properties {soundWaveAssetPtr=&/IP_Pickaxe_PalmTree_ImpactSweetener_08}}
 {id=WavePlayer_14, properties {soundWaveAssetPtr=&/IP_Pickaxe_PalmTree_ImpactSweetener_09}}
 {id=WavePlayer_2, properties {soundWaveAssetPtr=&/IP_Pickaxe_PalmTree_Impact_03}}
 {id=WavePlayer_3, properties {soundWaveAssetPtr=&/IP_Pickaxe_PalmTree_Impact_04}}
 {id=WavePlayer_4, properties {soundWaveAssetPtr=&/IP_Pickaxe_PalmTree_Impact_05}}
 {id=WavePlayer_5, properties {soundWaveAssetPtr=&/IP_Pickaxe_PalmTree_Impact_06}}
 {id=WavePlayer_6, properties {soundWaveAssetPtr=&/IP_Pickaxe_PalmTree_ImpactSweetener_01}}
 {id=WavePlayer_7, properties {soundWaveAssetPtr=&/IP_Pickaxe_PalmTree_ImpactSweetener_02}}
 {id=WavePlayer_8, properties {soundWaveAssetPtr=&/IP_Pickaxe_PalmTree_ImpactSweetener_03}}
 {id=WavePlayer_9, properties {soundWaveAssetPtr=&/IP_Pickaxe_PalmTree_ImpactSweetener_04}}
]
properties {
 attenuationSettings=/Game/Sounds/Attenuation/Fort_Pickaxe_Impact_Att
 audioPropertiesSheet=/BRCosmetics/Sounds/Settings/PropertySheets/Pickaxe/APS_Cosmetic_Pickaxe_Impact
 maxDistance=4300
 soundClassObject=&Fort_Impacts
 volumeMultiplier=0.525
 concurrencySet [null]
}
